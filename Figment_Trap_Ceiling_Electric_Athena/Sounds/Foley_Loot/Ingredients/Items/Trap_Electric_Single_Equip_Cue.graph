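# Sound cue for equipping a single electric ceiling trap item, classed as loot-collect audio.
# A mixer layers the trap's Ready sound with a quieter Spark sound.
# An enveloper at the root fades the blend out over a short window and slightly varies its pitch.
# A concurrency override caps instances at two by stopping the oldest, superseding the assigned concurrency set.
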SoundCue /Figment_Trap_Ceiling_Electric_Athena/Sounds/Foley_Loot/Ingredients/Items/Trap_Electric_Single_Equip_Cue
root=Enveloper_2
nodes [
 {
  id=Enveloper_2
  children [Mixer_1]
  properties {
   pitchMax=1.05
   pitchMin=0.95
   pitchCurve {editorCurveData {keys [{time=0}, {time=0.3}]}}
   volumeCurve {editorCurveData {keys [<at 0 1>, {arriveTangent=-0.2842, interpMode=RCIM_Cubic, leaveTangent=0.7849, tangentMode=RCTM_Break, time=0.2, value=1}, <at 0.65 0>]}}
  }
 }
 {id=Mixer_1, children [WavePlayer_43, WavePlayer_0], properties {inputVolume [1,0.3]}}
 {
  id=WavePlayer_0
  properties {soundWaveAssetPtr=/Figment_Trap_Ceiling_Electric_Athena/Sounds/Fort_Traps/Electric_Trap/CeilingTrap_Electric_Spark_001}
 }
 {id=WavePlayer_43, properties {soundWaveAssetPtr=/FNGameplayCues/Sounds/Fort_Traps/Electric_Trap/CeilingTrap_Electric_Ready_002}}
]
properties {
 bOverrideConcurrency=true
 maxDistance=2097152
 priority=0.5
 soundClassObject=&Fort_Loot_Collect
 volumeMultiplier=0.318
 concurrencyOverrides {maxCount=2, resolutionRule=StopOldest}
 concurrencySet [/Game/Sounds/Concurrency_Rules/OneMaximum_Owner_StopLowestPriority]
}
